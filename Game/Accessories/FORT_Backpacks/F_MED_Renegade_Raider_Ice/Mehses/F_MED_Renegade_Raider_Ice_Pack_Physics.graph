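# Physics asset for the Renegade Raider Ice backpack, defining physics bodies and joints for a rose with stems and leaves.
# Each dyn_ bone has a single capsule body, with constraints forming a tree rooted at dyn_StemBase.
# The base branches to the rose head with six head leaves, a short stem-leaf chain, and a second stem carrying three two-bone leaves.
# Tight swing and twist limits restrict every joint to small movements.
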PhysicsAsset /Game/Accessories/FORT_Backpacks/F_MED_Renegade_Raider_Ice/Mehses/F_MED_Renegade_Raider_Ice_Pack_Physics
bodies {
 dyn_StemBase [{shapeType=Capsule, length=18.4663, radius=1.1731, center=<-5.3976 2.4951 -7.5101>, rotation=<rot -17.6979 -90 34.6334>}]
 dyn_Stem_01 [{shapeType=Capsule, length=0.75, radius=1.75, shapeName=dyn_Stem_01_capsule, center=<0.2 0 0>, rotation=<rot -90 0 0>}]
 dyn_leaf01_02 [{shapeType=Capsule, length=0.75, radius=1.25, shapeName=dyn_leaf01_02_capsule, center=<0.2 0 0>, rotation=<rot 90 0 0>}]
 dyn_leaf01_03 [{shapeType=Capsule, length=0.75, radius=1.25, shapeName=dyn_leaf01_03_capsule, center=<0.2 0 0>, rotation=<rot 90 0 0>}]
 dyn_leaf02_01 [{shapeType=Capsule, length=1, radius=0.5, shapeName=dyn_leaf02_01_capsule, center=<0.2 0 0>, rotation=<rot 90 0 0>}]
 dyn_leaf02_02 [{shapeType=Capsule, length=1, radius=0.5, shapeName=dyn_leaf02_02_capsule, center=<0.2 0 0>, rotation=<rot 90 0 0>}]
 dyn_leaf03_01 [{shapeType=Capsule, length=0.5, radius=0.5, shapeName=dyn_leaf03_01_capsule, center=<0.2 0 0>, rotation=<rot 90 0 0>}]
 dyn_leaf03_02 [{shapeType=Capsule, length=1, radius=0.5, shapeName=dyn_leaf03_02_capsule, center=<0.2 0 0>, rotation=<rot 90 0 0>}]
 dyn_leaf04_01 [{shapeType=Capsule, length=1, radius=1, shapeName=dyn_leaf04_01_capsule, center=<0.2 0 0>, rotation=<rot 90 0 0>}]
 dyn_leaf04_02 [{shapeType=Capsule, length=0.75, radius=0.4, shapeName=dyn_leaf04_02_capsule, center=<0.2 0 0>, rotation=<rot 90 0 0>}]
 dyn_leafhead_01 [{shapeType=Capsule, length=0.9855, radius=0.5, shapeName=dyn_leafhead_01_capsule, center=<0.5 0 0>, rotation=<rot 0 90 90>}]
 dyn_leafhead_02 [{shapeType=Capsule, length=1.0439, radius=1.0099, shapeName=dyn_leafhead_02_capsule, center=<2 0 0>, rotation=<rot 90 0 0>}]
 dyn_leafhead_03 [{shapeType=Capsule, length=1.8516, radius=1.0124, shapeName=dyn_leafhead_03_capsule, center=<2 0 0>, rotation=<rot 90 0 0>}]
 dyn_leafhead_04 [{shapeType=Capsule, length=2.0103, radius=1, shapeName=dyn_leafhead_04_capsule, center=<2 0 0>, rotation=<rot 90 0 0>}]
 dyn_leafhead_05 [{shapeType=Capsule, length=0.9854, radius=1, shapeName=dyn_leafhead_05_capsule, center=<2 0 0>, rotation=<rot 90 0 0>}]
 dyn_leafhead_06 [{shapeType=Capsule, length=0.7156, radius=1, shapeName=dyn_leafhead_06_capsule, center=<2 0 0>, rotation=<rot 90 0 0>}]
 dyn_rosehead [{shapeType=Capsule, length=3.25, radius=7.5563, center=<10.5451 -0.0001 -1.1526>, rotation=<rot 90 0 0>}]
 dyn_stem_02 [{shapeType=Capsule, length=1, radius=1.5, shapeName=dyn_stem_02_capsule, center=<1.0105 0 0>, rotation=<rot 90 0 0>}]
 dyn_stem_03 [{shapeType=Capsule, length=0.75, radius=1, shapeName=dyn_stem_03_capsule, center=<0.2 0 0>, rotation=<rot 90 0 0>}]
}
constraints [
 {bone1Name=dyn_rosehead, bone2Name=dyn_StemBase, swing1LimitDegrees=12, swing2LimitDegrees=12, twistLimitDegrees=4}
 {bone1Name=dyn_leaf01_03, bone2Name=dyn_leaf01_02, swing1LimitDegrees=10, swing2LimitDegrees=10, twistLimitDegrees=2}
 {bone1Name=dyn_leaf01_02, bone2Name=dyn_Stem_01, swing1LimitDegrees=10, swing2LimitDegrees=10, twistLimitDegrees=3}
 {bone1Name=dyn_Stem_01, bone2Name=dyn_StemBase, swing1LimitDegrees=10, swing2LimitDegrees=10, twistLimitDegrees=3}
 {bone1Name=dyn_stem_02, bone2Name=dyn_StemBase, swing1LimitDegrees=10, swing2LimitDegrees=8, twistLimitDegrees=3}
 {bone1Name=dyn_stem_03, bone2Name=dyn_stem_02, swing1LimitDegrees=10, swing2LimitDegrees=10, twistLimitDegrees=3}
 {bone1Name=dyn_leaf02_01, bone2Name=dyn_stem_03, swing1LimitDegrees=10, swing2LimitDegrees=7, twistLimitDegrees=3}
 {bone1Name=dyn_leaf02_02, bone2Name=dyn_leaf02_01, swing1LimitDegrees=10, swing2LimitDegrees=10, twistLimitDegrees=3}
 {bone1Name=dyn_leaf03_01, bone2Name=dyn_stem_03, swing1LimitDegrees=10, swing2LimitDegrees=7, twistLimitDegrees=3}
 {bone1Name=dyn_leaf03_02, bone2Name=dyn_leaf03_01, swing1LimitDegrees=10, swing2LimitDegrees=10, twistLimitDegrees=3}
 {bone1Name=dyn_leaf04_01, bone2Name=dyn_stem_03, swing1LimitDegrees=10, swing2LimitDegrees=7, twistLimitDegrees=3}
 {bone1Name=dyn_leaf04_02, bone2Name=dyn_leaf04_01, swing1LimitDegrees=10, swing2LimitDegrees=5, twistLimitDegrees=3}
 {bone1Name=dyn_leafhead_01, bone2Name=dyn_rosehead, swing1LimitDegrees=3, swing2LimitDegrees=10, twistLimitDegrees=2}
 {bone1Name=dyn_leafhead_02, bone2Name=dyn_rosehead, swing1LimitDegrees=3, swing2LimitDegrees=10, twistLimitDegrees=2}
 {bone1Name=dyn_leafhead_03, bone2Name=dyn_rosehead, swing1LimitDegrees=3, swing2LimitDegrees=10, twistLimitDegrees=2}
 {bone1Name=dyn_leafhead_04, bone2Name=dyn_rosehead, swing1LimitDegrees=3, swing2LimitDegrees=10, twistLimitDegrees=2}
 {bone1Name=dyn_leafhead_05, bone2Name=dyn_rosehead, swing1LimitDegrees=3, swing2LimitDegrees=10, twistLimitDegrees=2}
 {bone1Name=dyn_leafhead_06, bone2Name=dyn_rosehead, swing1LimitDegrees=3, swing2LimitDegrees=10, twistLimitDegrees=2}
]
